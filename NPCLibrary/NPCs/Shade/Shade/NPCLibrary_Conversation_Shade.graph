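# Conversation database for the NPC Shade, giving the character's greeting and the hand-off to shared dialogue branches.
# The Shade entry point leads to one Speech node with a default line about driving jobs.
# Contextual lines replace it for bar count, a first-ever conversation or the Street Racer Cobra outfit, ranked by priority curve rows.
# The speech then offers links by tag to the shared NPCLibrary Quests and BasicServices conversations.
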ConversationDatabase /NPCLibrary/NPCs/Shade/Shade/NPCLibrary_Conversation_Shade
entry_points [{tag=Conversation.Entry.Shade, destinations [#1]}]
nodes [
 ConversationEntryPointNode_7 {
  id=#1
  class=ConversationEntryPointNode
  outputs [#2]
  properties {entryTag {tagName=Conversation.Entry.Shade}}
 }
 Speech_C_4 {
  id=#2
  class=Speech_C
  outputs [#3, #4]
  properties {
   nodeName=""
   generalConfig {
    defaultMessage="You hear about any jobs that need a driver, you let me know yeah?"
    contextualMessages [
     {
      message="Must've been working some <b>choice jobs</> to stuff your pockets with that much cash. Noice."
      contextRequirements [{requirement {class=FortControllerRequirement_HasBars, properties {thresholdTestConfig {requirement=GreaterThanOrEqual}}}}]
      priority {curve {rowName=Tandem.Speech.Priority.HighTotalBars}}
     }
     {
      message="What, are you looking to race or something?"
      contextRequirements [{requirement {class=FortControllerRequirement_IsFirstEverConversationWithNPC}}]
      priority {curve {rowName=Tandem.Speech.Priority.FirstEverInteraction}}
     }
     {
      message="We really gotta get a <b>Cobra Crew handshake</> or something."
      contextRequirements [{requirement {
         class=FortControllerRequirement_HasCID
         properties {allowedCIds [/BRCosmetics/Athena/Items/Cosmetics/Characters/CID_177_Athena_Commando_M_StreetRacerCobra]}
        }}]
      priority {curve {rowName=Tandem.Speech.Priority.OutfitReaction}}
     }
    ]
   }
  }
 }
 ConversationLinkNode_0 {id=#4, class=ConversationLinkNode, properties {remoteEntryTag {tagName=Conversation.Entry.BasicServices}}}
 ConversationLinkNode_1 {id=#3, class=ConversationLinkNode, properties {remoteEntryTag {tagName=Conversation.Entry.NPCLibrary.Quests}}}
]
